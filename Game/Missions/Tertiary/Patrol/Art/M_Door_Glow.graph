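# M_Door_Glow is a two-sided, mesh-particle-compatible material for a glow, by its name on a door in Patrol mission art.
# The UV V coordinate is shaped by Exponent and Multiply parameters, clamped and inverted, making a gradient that fades as V rises.
# That gradient scales particle color into Emissive, while particle alpha feeds Opacity.
# The resulting material attributes pass through the MF_CreativeMoveTool function.
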Material /Game/Missions/Tertiary/Patrol/Art/M_Door_Glow
graph {
 properties {bUseMaterialAttributes=true, bUsedWithMeshParticles=true, pixelDepthOffsetMode=PDOM_Legacy, twoSided=true, usedShadingModels=Default Lit}
 nodes [
  n1 = ParticleColor()
  n2 = TextureCoordinate()
  n3 = Clamp(None=Power(Base=ComponentMask(None=@n2) {g=true}, Exp=ScalarParameter() {defaultValue=1, parameterName=Exponent}) {constExponent=2.2})
  n4 = ScalarParameter() {defaultValue=1, parameterName=Multiply}
  n5 = OneMinus(None=Clamp(None=Multiply(A=@n3, B=@n4)))
  MaterialFunctionCall(MA In=MakeMaterialAttributes!(
    BaseColor
    Metallic
    Specular
    Roughness
    Anisotropy
    EmissiveColor=Multiply(A=@n1.RGB, B=@n5)
    Opacity=@n1.A
    OpacityMask
    Normal
    Tangent
    WorldPositionOffset
    SubsurfaceColor
    ClearCoat
    ClearCoatRoughness
    AmbientOcclusion
    Refraction
    CustomizedUVs_0
    CustomizedUVs_1
    CustomizedUVs_2
    CustomizedUVs_3
    CustomizedUVs_4
    CustomizedUVs_5
    CustomizedUVs_6
    CustomizedUVs_7
    PixelDepthOffset
    ShadingModel
    Displacement
   )) {materialFunction=&MF_CreativeMoveTool}
 ]
}
